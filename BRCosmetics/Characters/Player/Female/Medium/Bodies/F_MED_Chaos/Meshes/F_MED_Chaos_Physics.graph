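# Physics asset for the F_MED_Chaos female medium player body cosmetic, covering only its four pouches, two per side.
# Each pouch has a tiny capsule on its master bone and a larger offset capsule on the child pouch bone.
# A constraint joins each pair with tight twist and first-swing limits and a wider second swing, so the pouch moves mostly along one axis.
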PhysicsAsset /BRCosmetics/Characters/Player/Female/Medium/Bodies/F_MED_Chaos/Meshes/F_MED_Chaos_Physics
bodies {
 Dyn_Master_Pouch1_l [{shapeType=Capsule, length=0.5, radius=0.5, shapeName=Dyn_Master_Pouch1_l_capsule}]
 Dyn_Master_Pouch1_r [{shapeType=Capsule, length=0.5, radius=0.5, shapeName=Dyn_Master_Pouch1_r_capsule}]
 Dyn_Master_Pouch2_l [{shapeType=Capsule, length=0.5, radius=0.5, shapeName=Dyn_Master_Pouch2_l_capsule}]
 Dyn_Master_Pouch2_r [{shapeType=Capsule, length=0.5, radius=0.5, shapeName=Dyn_Master_Pouch2_r_capsule}]
 Dyn_Pouch1_l1 [{shapeType=Capsule, length=8, radius=4, shapeName=Dyn_Pouch1_l1_capsule, center=<1.2252 -3.8341 0>, rotation=<rot 90 0 0>}]
 Dyn_Pouch1_r1 [{shapeType=Capsule, length=8, radius=4, shapeName=Dyn_Pouch1_r1_capsule, center=<1.2252 -3.8341 0>, rotation=<rot 90 0 0>}]
 Dyn_Pouch2_l1 [{shapeType=Capsule, length=8, radius=4, shapeName=Dyn_Pouch2_l1_capsule, center=<1.2252 -3.8341 0>, rotation=<rot 90 0 0>}]
 Dyn_Pouch2_r1 [{shapeType=Capsule, length=8, radius=4, shapeName=Dyn_Pouch2_r1_capsule, center=<1.2252 -3.8341 0>, rotation=<rot 90 0 0>}]
}
constraints [
 {bone1Name=Dyn_Pouch1_r1, bone2Name=Dyn_Master_Pouch1_r, swing1LimitDegrees=2, swing2LimitDegrees=20, twistLimitDegrees=1}
 {bone1Name=Dyn_Pouch2_r1, bone2Name=Dyn_Master_Pouch2_r, swing1LimitDegrees=2, swing2LimitDegrees=20, twistLimitDegrees=1}
 {bone1Name=Dyn_Pouch2_l1, bone2Name=Dyn_Master_Pouch2_l, swing1LimitDegrees=2, swing2LimitDegrees=20, twistLimitDegrees=1}
 {bone1Name=Dyn_Pouch1_l1, bone2Name=Dyn_Master_Pouch1_l, swing1LimitDegrees=2, swing2LimitDegrees=20, twistLimitDegrees=1}
]
