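# Post-process style AnimBP for the F_MED_Chaos female body, layering pouch secondary motion onto the parent mesh's copied pose.
# After a deform Control Rig, additive bone rotations nudge hip pouches while jumping or falling, and the left pair also while crouching.
# Constraints tie shoulder pouches to clavicles and blend hip pouch rotation between pelvis and thighs, feeding a pelvis-space RigidBody simulation.
# An extra left-pouch constraint pair is disconnected from the output.
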AnimBlueprint /BRCosmetics/Characters/Player/Female/Medium/Bodies/F_MED_Chaos/Meshes/F_MED_Chaos_AnimBP
target_skeleton=&/F_MED_Chaos_Skeleton
graphs [
 AnimGraph {nodes [
   n1 = "Constraint - Dyn_Master_Pouch1_l (R)"(Alpha=1, ConstraintWeights_0=0.3) {
    alphaBoolBlend {blendOption=Linear}
    boneToModify {boneName=Dyn_Master_Pouch1_l}
    constraintSetup [{transformType=Rotation, targetBone {boneName=pelvis}}]
    constraintWeights [1]
   }
   "Constraint - Dyn_Master_Pouch2_l (R)"(ComponentPose=@n1, Alpha=1, ConstraintWeights_0=0.3) {
    alphaBoolBlend {blendOption=Linear}
    boneToModify {boneName=Dyn_Master_Pouch2_l}
    constraintSetup [{transformType=Rotation, targetBone {boneName=pelvis}}]
    constraintWeights [1]
   }
   n2 = Control Rig(Source=Copy Pose From Mesh() {bUseAttachedParent=true}) {
    alphaBoolBlend {blendOption=Linear}
    controlRigAssetReference {blueprintRigClass=&/F_MED_DeformRig_Chaos_ControlRig.F_MED_DeformRig_Chaos_ControlRig_C}
   }
   n3 = Get bIsCrouching()
   n4 = "Transform (Modify) Bone - Bone: Dyn_Master_Pouch1_l"!(ComponentPose=Local To Component(LocalPose=@n2), bAlphaBoolEnabled:bool=@n3, Translation="0,0,0", Rotation="0,12,0", Scale="1,1,1") -> (Pose) {
    alphaInputType=Bool
    rotationMode=BMM_Additive
    rotationSpace=BCS_BoneSpace
    scaleMode=BMM_Additive
    scaleSpace=BCS_BoneSpace
    translationMode=BMM_Additive
    translationSpace=BCS_BoneSpace
    alphaBoolBlend {blendOption=Linear}
    boneToModify {boneName=Dyn_Master_Pouch1_l}
    rotation {yaw=12}
   }
   n5 = "Transform (Modify) Bone - Bone: Dyn_Master_Pouch2_l"!(ComponentPose=@n4, bAlphaBoolEnabled:bool=@n3, Translation="0,0,0", Rotation="0,12,0", Scale="1,1,1") -> (Pose) {
    alphaInputType=Bool
    rotationMode=BMM_Additive
    rotationSpace=BCS_BoneSpace
    scaleMode=BMM_Additive
    scaleSpace=BCS_BoneSpace
    translationMode=BMM_Additive
    translationSpace=BCS_BoneSpace
    alphaBoolBlend {blendOption=Linear}
    boneToModify {boneName=Dyn_Master_Pouch2_l}
    rotation {yaw=12}
   }
   n6 = Get bIsJumping()
   n7 = "Transform (Modify) Bone - Bone: Dyn_Master_Pouch1_l"(ComponentPose=@n5, bAlphaBoolEnabled=@n6, Translation="0,0,0", Rotation="0,12,0", Scale="1,1,1") {
    alphaInputType=Bool
    rotationMode=BMM_Additive
    rotationSpace=BCS_BoneSpace
    scaleMode=BMM_Additive
    scaleSpace=BCS_BoneSpace
    translationMode=BMM_Additive
    translationSpace=BCS_BoneSpace
    alphaBoolBlend {blendInTime=0.5, blendOption=Linear}
    boneToModify {boneName=Dyn_Master_Pouch1_l}
    rotation {yaw=12}
   }
   n8 = "Transform (Modify) Bone - Bone: Dyn_Master_Pouch2_l"(ComponentPose=@n7, bAlphaBoolEnabled=@n6, Translation="0,0,0", Rotation="0,12,0", Scale="1,1,1") {
    alphaInputType=Bool
    rotationMode=BMM_Additive
    rotationSpace=BCS_BoneSpace
    scaleMode=BMM_Additive
    scaleSpace=BCS_BoneSpace
    translationMode=BMM_Additive
    translationSpace=BCS_BoneSpace
    alphaBoolBlend {blendInTime=0.5, blendOption=Linear}
    boneToModify {boneName=Dyn_Master_Pouch2_l}
    rotation {yaw=12}
   }
   n9 = Get bIsFalling()
   n10 = "Transform (Modify) Bone - Bone: Dyn_Master_Pouch1_l"(ComponentPose=@n8, bAlphaBoolEnabled=@n9, Translation="0,0,0", Rotation="0,12,0", Scale="1,1,1") {
    alphaInputType=Bool
    rotationMode=BMM_Additive
    rotationSpace=BCS_BoneSpace
    scaleMode=BMM_Additive
    scaleSpace=BCS_BoneSpace
    translationMode=BMM_Additive
    translationSpace=BCS_BoneSpace
    alphaBoolBlend {blendOption=Linear, blendOutTime=0.5}
    boneToModify {boneName=Dyn_Master_Pouch1_l}
    rotation {yaw=12}
   }
   n11 = "Transform (Modify) Bone - Bone: Dyn_Master_Pouch2_l"(ComponentPose=@n10, bAlphaBoolEnabled=@n9, Translation="0,0,0", Rotation="0,12,0", Scale="1,1,1") {
    alphaInputType=Bool
    rotationMode=BMM_Additive
    rotationSpace=BCS_BoneSpace
    scaleMode=BMM_Additive
    scaleSpace=BCS_BoneSpace
    translationMode=BMM_Additive
    translationSpace=BCS_BoneSpace
    alphaBoolBlend {blendOption=Linear, blendOutTime=0.5}
    boneToModify {boneName=Dyn_Master_Pouch2_l}
    rotation {yaw=12}
   }
   n12 = "Transform (Modify) Bone - Bone: Dyn_Master_Pouch1_r"!(ComponentPose=@n11, bAlphaBoolEnabled:bool=@n6, Translation="0,0,0", Rotation="0,12,0", Scale="1,1,1") -> (Pose) {
    alphaInputType=Bool
    rotationMode=BMM_Additive
    rotationSpace=BCS_BoneSpace
    scaleMode=BMM_Additive
    scaleSpace=BCS_BoneSpace
    translationMode=BMM_Additive
    translationSpace=BCS_BoneSpace
    alphaBoolBlend {blendInTime=0.5, blendOption=Linear}
    boneToModify {boneName=Dyn_Master_Pouch1_r}
    rotation {yaw=12}
   }
   n13 = "Transform (Modify) Bone - Bone: Dyn_Master_Pouch2_r"!(ComponentPose=@n12, bAlphaBoolEnabled:bool=@n6, Translation="0,0,0", Rotation="0,12,0", Scale="1,1,1") -> (Pose) {
    alphaInputType=Bool
    rotationMode=BMM_Additive
    rotationSpace=BCS_BoneSpace
    scaleMode=BMM_Additive
    scaleSpace=BCS_BoneSpace
    translationMode=BMM_Additive
    translationSpace=BCS_BoneSpace
    alphaBoolBlend {blendInTime=0.5, blendOption=Linear}
    boneToModify {boneName=Dyn_Master_Pouch2_r}
    rotation {yaw=12}
   }
   n14 = "Transform (Modify) Bone - Bone: Dyn_Master_Pouch1_r"(ComponentPose=@n13, bAlphaBoolEnabled=@n9, Translation="0,0,0", Rotation="0,12,0", Scale="1,1,1") {
    alphaInputType=Bool
    rotationMode=BMM_Additive
    rotationSpace=BCS_BoneSpace
    scaleMode=BMM_Additive
    scaleSpace=BCS_BoneSpace
    translationMode=BMM_Additive
    translationSpace=BCS_BoneSpace
    alphaBoolBlend {blendOption=Linear, blendOutTime=0.5}
    boneToModify {boneName=Dyn_Master_Pouch1_r}
    rotation {yaw=12}
   }
   n15 = "Transform (Modify) Bone - Bone: Dyn_Master_Pouch2_r"(ComponentPose=@n14, bAlphaBoolEnabled=@n9, Translation="0,0,0", Rotation="0,12,0", Scale="1,1,1") {
    alphaInputType=Bool
    rotationMode=BMM_Additive
    rotationSpace=BCS_BoneSpace
    scaleMode=BMM_Additive
    scaleSpace=BCS_BoneSpace
    translationMode=BMM_Additive
    translationSpace=BCS_BoneSpace
    alphaBoolBlend {blendOption=Linear, blendOutTime=0.5}
    boneToModify {boneName=Dyn_Master_Pouch2_r}
    rotation {yaw=12}
   }
   n16 = "Constraint - Dyn_Master_ShoulderPouch_r (P)"(ComponentPose=@n15, Alpha=1, ConstraintWeights_0=1) {
    alphaBoolBlend {blendOption=Linear}
    boneToModify {boneName=Dyn_Master_ShoulderPouch_r}
    constraintSetup [{targetBone {boneName=clavicle_r}}]
    constraintWeights [1]
   }
   n17 = "Constraint - Dyn_Master_ShoulderPouch_l (P)"(ComponentPose=@n16, Alpha=1, ConstraintWeights_0=1) {
    alphaBoolBlend {blendOption=Linear}
    boneToModify {boneName=Dyn_Master_ShoulderPouch_l}
    constraintSetup [{targetBone {boneName=clavicle_l}}]
    constraintWeights [1]
   }
   n18 = "Constraint - Dyn_Master_Pouch1_l (RR)"!(ComponentPose=@n17, Alpha=1, ConstraintWeights_0=0.3, ConstraintWeights_1:float=0.7) -> (Pose) {
    alphaBoolBlend {blendOption=Linear}
    boneToModify {boneName=Dyn_Master_Pouch1_l}
    constraintSetup [{transformType=Rotation, targetBone {boneName=pelvis}}, {transformType=Rotation, targetBone {boneName=thigh_l}}]
    constraintWeights [1,1]
   }
   n19 = "Constraint - Dyn_Master_Pouch2_l (RR)"!(ComponentPose=@n18, Alpha=1, ConstraintWeights_0=0.3, ConstraintWeights_1:float=0.7) -> (Pose) {
    alphaBoolBlend {blendOption=Linear}
    boneToModify {boneName=Dyn_Master_Pouch2_l}
    constraintSetup [{transformType=Rotation, targetBone {boneName=pelvis}}, {transformType=Rotation, targetBone {boneName=thigh_l}}]
    constraintWeights [1,1]
   }
   n20 = "Constraint - Dyn_Master_Pouch1_r (RR)"!(ComponentPose=@n19, Alpha=1, ConstraintWeights_0=0.3, ConstraintWeights_1:float=0.7) -> (Pose) {
    alphaBoolBlend {blendOption=Linear}
    boneToModify {boneName=Dyn_Master_Pouch1_r}
    constraintSetup [{transformType=Rotation, targetBone {boneName=pelvis}}, {transformType=Rotation, targetBone {boneName=thigh_r}}]
    constraintWeights [1,1]
   }
   n21 = "Constraint - Dyn_Master_Pouch2_r (RR)"!(ComponentPose=@n20, Alpha=1, ConstraintWeights_0=0.3, ConstraintWeights_1:float=0.7) -> (Pose) {
    alphaBoolBlend {blendOption=Linear}
    boneToModify {boneName=Dyn_Master_Pouch2_r}
    constraintSetup [{transformType=Rotation, targetBone {boneName=pelvis}}, {transformType=Rotation, targetBone {boneName=thigh_r}}]
    constraintWeights [1,1]
   }
   n22 = RigidBody(ComponentPose=@n21) {
    bClampLinearTranslationLimitToRefPose=true
    bOverrideWorldGravity=true
    overridePhysicsAsset=&/F_MED_Chaos_Physics
    simulationSpace=BaseBoneSpace
    alphaBoolBlend {blendOption=Linear}
    baseBoneRef {boneName=pelvis}
    componentAppliedLinearAccClamp=<3000 3000 3000>
    componentLinearAccScale=<0.5 0.5 0.5>
    componentLinearVelScale=<0.5 0.5 0.5>
    overrideWorldGravity {z=-980}
   }
   Output Pose(Result=Component To Local(ComponentPose=@n22)) {layerGroup=None}
  ]}
 EventGraph {nodes [Event BlueprintUpdateAnimation()]}
]
variables []
