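# Physics asset for the F_MED_Chaos female face accessory mesh, setting up dynamic hair against capsule bodies for head, upper spine, clavicles and upper arms.
# Three five-segment back chains and a short front-right chain hang from master bones attached to the head.
# Chain joints allow almost no twist, with swing tightening down the back chains.
# Angularly free cross-constraints tie the middle back chain to both side chains at segments 3 and 5.
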PhysicsAsset /BRCosmetics/Characters/Player/Female/Medium/Bodies/F_MED_Chaos/Meshes/Parts/F_MED_Chaos_FaceAcc_Physics
bodies {
 Dyn_Front_r1 [{shapeType=Capsule, length=4, radius=3, shapeName=Dyn_Front_r1_capsule, center=<1.6274 0 0>, rotation=<rot 90 -180 -180>}]
 Dyn_Front_r2 [{shapeType=Capsule, length=2, radius=3, shapeName=Dyn_Front_r2_capsule, center=<1.1055 0 0>, rotation=<rot 90 0 0>}]
 Dyn_Front_r3 [{shapeType=Capsule, length=3, radius=3, shapeName=Dyn_Front_r3_capsule, center=<1.4801 0 0>, rotation=<rot 90 0 0>}]
 Dyn_LeftBack_Hair1 [{shapeType=Capsule, length=15, radius=3, shapeName=Dyn_LeftBack_Hair1_capsule, center=<6.7033 -1.7017 0>, rotation=<rot 90 0 9.9994>}]
 Dyn_LeftBack_Hair2 [{shapeType=Capsule, length=8, radius=4, shapeName=Dyn_LeftBack_Hair2_capsule, center=<4.8573 -0.0263 2.0402>, rotation=<rot 90 180 -160.0005>}]
 Dyn_LeftBack_Hair3 [{shapeType=Capsule, length=8, radius=4, shapeName=Dyn_LeftBack_Hair3_capsule, center=<4.2584 -1.0228 0>, rotation=<rot 90 0 0>}]
 Dyn_LeftBack_Hair4 [{shapeType=Capsule, length=6, radius=5, shapeName=Dyn_LeftBack_Hair4_capsule, center=<3.0975 -2.4968 1.5336>, rotation=<rot 90 0 0>}]
 Dyn_LeftBack_Hair5 [{
   shapeType=Capsule
   length=10
   radius=4
   shapeName=Dyn_LeftBack_Hair5_capsule
   center=<2.0198 -0.0502 0.0975>
   rotation=<rot 50.0001 0.0001 -10>
  }]
 Dyn_Master_Front_r [{shapeType=Capsule, length=0.5, radius=0.5, shapeName=Dyn_Master_Front_r_capsule}]
 Dyn_Master_LeftBack_Hair [{shapeType=Capsule, length=0.5, radius=0.5, shapeName=Dyn_Master_LeftBack_Hair_capsule}]
 Dyn_Master_MiddleBack_Hair [{shapeType=Capsule, length=0.5, radius=0.5, shapeName=Dyn_Master_MiddleBack_Hair_capsule}]
 Dyn_Master_RightBack_Hair [{shapeType=Capsule, length=0.5, radius=0.5, shapeName=Dyn_Master_RightBack_Hair_capsule}]
 Dyn_MiddleBack_Hair1 [{shapeType=Capsule, length=15, radius=3, shapeName=Dyn_MiddleBack_Hair1_capsule, center=<6.6737 -1.6684 0>, rotation=<rot 90 0 9.9994>}]
 Dyn_MiddleBack_Hair2 [{shapeType=Capsule, length=8, radius=3.5, shapeName=Dyn_MiddleBack_Hair2_capsule, center=<4.5 -0.5572 0>, rotation=<rot 90 0 0>}]
 Dyn_MiddleBack_Hair3 [{shapeType=Capsule, length=8, radius=3.5, shapeName=Dyn_MiddleBack_Hair3_capsule, center=<4.2584 -2.9128 0>}]
 Dyn_MiddleBack_Hair4 [{shapeType=Capsule, length=6, radius=5, shapeName=Dyn_MiddleBack_Hair4_capsule, center=<3.0975 -4.603 0>, rotation=<rot 90 0 0>}]
 Dyn_MiddleBack_Hair5 [{shapeType=Capsule, length=10, radius=4, shapeName=Dyn_MiddleBack_Hair5_capsule, center=<7 1 0>, rotation=<rot 0 -159.9997 -180>}]
 Dyn_RightBack_Hair1 [{shapeType=Capsule, length=15, radius=3, shapeName=Dyn_RightBack_Hair1_capsule, center=<6.913 -0.5124 0>, rotation=<rot 90 0 9.9994>}]
 Dyn_RightBack_Hair2 [{shapeType=Capsule, length=8, radius=4, shapeName=Dyn_RightBack_Hair2_capsule, center=<4.5 2.436 0>, rotation=<rot 90 0 0>}]
 Dyn_RightBack_Hair3 [{shapeType=Capsule, length=8, radius=4, shapeName=Dyn_RightBack_Hair3_capsule, center=<4.2584 -1.2314 0>, rotation=<rot 90 0 0>}]
 Dyn_RightBack_Hair4 [{shapeType=Capsule, length=6, radius=5, shapeName=Dyn_RightBack_Hair4_capsule, center=<3.0975 -1.7198 -1.2669>, rotation=<rot 90 0 0>}]
 Dyn_RightBack_Hair5 [{
   shapeType=Capsule
   length=10
   radius=4
   shapeName=Dyn_RightBack_Hair5_capsule
   center=<4.5918 2.6489 0.1514>
   rotation=<rot 50.0002 -159.9994 -169.9999>
  }]
 clavicle_l [{shapeType=Capsule, length=10, radius=5, shapeName=clavicle_l_capsule, center=<8.0962 0.5059 -0.6753>, rotation=<rot 80 -0.0009 -0.001>}]
 clavicle_r [{shapeType=Capsule, length=10, radius=5, shapeName=clavicle_r_capsule, center=<-6.6353 0 0.3423>, rotation=<rot 80 -0.0005 -0.0006>}]
 head [{shapeType=Capsule, length=5.5, radius=6.75, shapeName=head_capsule, center=<-0.9274 3.4114 0>, rotation=<rot 90 0 0>}]
 spine_05 [
  {shapeType=Capsule, length=6.5, radius=10, shapeName=spine_05_capsule, center=<-4.244 1.3532 0>, rotation=<rot 0 -90 0>}
  {shapeType=Capsule, length=6.5, radius=10, shapeName=spine_05_capsule, center=<1.0151 0.6141 0>}
  {shapeType=Capsule, length=6.5, radius=10, shapeName=spine_05_capsule, center=<-10 0.5 0>}
 ]
 upperarm_l [{shapeType=Capsule, length=15, radius=5, shapeName=upperarm_l_capsule, center=<10 0 0>, rotation=<rot 90 0 0>}]
 upperarm_r [{shapeType=Capsule, length=15, radius=5, shapeName=upperarm_r_capsule, center=<-10 0 0>, rotation=<rot 90 0 0>}]
}
constraints [
 {
  bone1Name=Dyn_Master_MiddleBack_Hair
  bone2Name=head
  swing1LimitDegrees=45
  swing1Motion=Free
  swing2LimitDegrees=45
  swing2Motion=Free
  twistLimitDegrees=45
  twistMotion=Free
 }
 {
  bone1Name=Dyn_Master_LeftBack_Hair
  bone2Name=head
  swing1LimitDegrees=45
  swing1Motion=Free
  swing2LimitDegrees=45
  swing2Motion=Free
  twistLimitDegrees=45
  twistMotion=Free
 }
 {
  bone1Name=Dyn_Master_RightBack_Hair
  bone2Name=head
  swing1LimitDegrees=45
  swing1Motion=Free
  swing2LimitDegrees=45
  swing2Motion=Free
  twistLimitDegrees=45
  twistMotion=Free
 }
 {
  bone1Name=Dyn_Master_Front_r
  bone2Name=head
  swing1LimitDegrees=45
  swing1Motion=Free
  swing2LimitDegrees=45
  swing2Motion=Free
  twistLimitDegrees=45
  twistMotion=Free
 }
 {bone1Name=Dyn_MiddleBack_Hair1, bone2Name=Dyn_Master_MiddleBack_Hair, swing1LimitDegrees=80, swing2LimitDegrees=30, twistLimitDegrees=1}
 {bone1Name=Dyn_MiddleBack_Hair2, bone2Name=Dyn_MiddleBack_Hair1, swing1LimitDegrees=60, swing2LimitDegrees=30, twistLimitDegrees=1}
 {bone1Name=Dyn_MiddleBack_Hair3, bone2Name=Dyn_MiddleBack_Hair2, swing1LimitDegrees=40, swing2LimitDegrees=30, twistLimitDegrees=1}
 {bone1Name=Dyn_MiddleBack_Hair4, bone2Name=Dyn_MiddleBack_Hair3, swing1LimitDegrees=30, swing2LimitDegrees=20, twistLimitDegrees=1}
 {bone1Name=Dyn_MiddleBack_Hair5, bone2Name=Dyn_MiddleBack_Hair4, swing1LimitDegrees=30, swing2LimitDegrees=20, twistLimitDegrees=1}
 {bone1Name=Dyn_LeftBack_Hair1, bone2Name=Dyn_Master_LeftBack_Hair, swing1LimitDegrees=80, swing2LimitDegrees=30, twistLimitDegrees=1}
 {bone1Name=Dyn_LeftBack_Hair2, bone2Name=Dyn_LeftBack_Hair1, swing1LimitDegrees=60, swing2LimitDegrees=30, twistLimitDegrees=1}
 {bone1Name=Dyn_LeftBack_Hair3, bone2Name=Dyn_LeftBack_Hair2, swing1LimitDegrees=40, swing2LimitDegrees=30, twistLimitDegrees=1}
 {bone1Name=Dyn_LeftBack_Hair4, bone2Name=Dyn_LeftBack_Hair3, swing1LimitDegrees=30, swing2LimitDegrees=20, twistLimitDegrees=1}
 {bone1Name=Dyn_LeftBack_Hair5, bone2Name=Dyn_LeftBack_Hair4, swing1LimitDegrees=30, swing2LimitDegrees=20, twistLimitDegrees=1}
 {bone1Name=Dyn_RightBack_Hair1, bone2Name=Dyn_Master_RightBack_Hair, swing1LimitDegrees=80, swing2LimitDegrees=30, twistLimitDegrees=1}
 {bone1Name=Dyn_RightBack_Hair2, bone2Name=Dyn_RightBack_Hair1, swing1LimitDegrees=60, swing2LimitDegrees=30, twistLimitDegrees=1}
 {bone1Name=Dyn_RightBack_Hair3, bone2Name=Dyn_RightBack_Hair2, swing1LimitDegrees=40, swing2LimitDegrees=30, twistLimitDegrees=1}
 {bone1Name=Dyn_RightBack_Hair4, bone2Name=Dyn_RightBack_Hair3, swing1LimitDegrees=30, swing2LimitDegrees=20, twistLimitDegrees=1}
 {bone1Name=Dyn_RightBack_Hair5, bone2Name=Dyn_RightBack_Hair4, swing1LimitDegrees=30, swing2LimitDegrees=20, twistLimitDegrees=1}
 {bone1Name=Dyn_Front_r1, bone2Name=Dyn_Master_Front_r, swing1LimitDegrees=10, swing2LimitDegrees=30, twistLimitDegrees=1}
 {bone1Name=Dyn_Front_r2, bone2Name=Dyn_Front_r1, swing1LimitDegrees=10, swing2LimitDegrees=30, twistLimitDegrees=1}
 {bone1Name=Dyn_Front_r3, bone2Name=Dyn_Front_r2, swing1LimitDegrees=30, swing2LimitDegrees=30, twistLimitDegrees=1}
 {
  bone1Name=clavicle_l
  bone2Name=spine_05
  swing1LimitDegrees=45
  swing1Motion=Free
  swing2LimitDegrees=45
  swing2Motion=Free
  twistLimitDegrees=45
  twistMotion=Free
 }
 {
  bone1Name=head
  bone2Name=spine_05
  swing1LimitDegrees=45
  swing1Motion=Free
  swing2LimitDegrees=45
  swing2Motion=Free
  twistLimitDegrees=45
  twistMotion=Free
 }
 {
  bone1Name=clavicle_r
  bone2Name=spine_05
  swing1LimitDegrees=45
  swing1Motion=Free
  swing2LimitDegrees=45
  swing2Motion=Free
  twistLimitDegrees=45
  twistMotion=Free
 }
 {
  bone1Name=Dyn_MiddleBack_Hair5
  bone2Name=Dyn_RightBack_Hair5
  swing1LimitDegrees=45
  swing1Motion=Free
  swing2LimitDegrees=45
  swing2Motion=Free
  twistLimitDegrees=45
  twistMotion=Free
 }
 {
  bone1Name=Dyn_LeftBack_Hair5
  bone2Name=Dyn_MiddleBack_Hair5
  swing1LimitDegrees=45
  swing1Motion=Free
  swing2LimitDegrees=45
  swing2Motion=Free
  twistLimitDegrees=45
  twistMotion=Free
 }
 {
  bone1Name=Dyn_MiddleBack_Hair3
  bone2Name=Dyn_RightBack_Hair3
  swing1LimitDegrees=45
  swing1Motion=Free
  swing2LimitDegrees=45
  swing2Motion=Free
  twistLimitDegrees=45
  twistMotion=Free
 }
 {
  bone1Name=Dyn_MiddleBack_Hair3
  bone2Name=Dyn_LeftBack_Hair3
  swing1LimitDegrees=45
  swing1Motion=Free
  swing2LimitDegrees=45
  swing2Motion=Free
  twistLimitDegrees=45
  twistMotion=Free
 }
 {
  bone1Name=upperarm_l
  bone2Name=clavicle_l
  swing1LimitDegrees=45
  swing1Motion=Free
  swing2LimitDegrees=45
  swing2Motion=Free
  twistLimitDegrees=45
  twistMotion=Free
 }
 {
  bone1Name=upperarm_r
  bone2Name=clavicle_r
  swing1LimitDegrees=45
  swing1Motion=Free
  swing2LimitDegrees=45
  swing2Motion=Free
  twistLimitDegrees=45
  twistMotion=Free
 }
]
