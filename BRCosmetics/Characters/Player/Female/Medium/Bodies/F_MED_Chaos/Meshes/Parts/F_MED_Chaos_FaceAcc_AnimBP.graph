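# Post-process-style AnimBP for the F_MED_Chaos face accessory mesh, animating its back and side hair.
# It copies the parent pose with Body_Mesh shoulders layered in, then applies master-bone constraints and head-aim/targeting offsets.
# Two RigidBody nodes switch physics assets by backbling presence, detected when customization loads.
# Position constraints and bone offsets then pose middle hair under backbling, and left/right hair when rigid bodies are disabled, adjusted for skydiving or emotes.
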AnimBlueprint /BRCosmetics/Characters/Player/Female/Medium/Bodies/F_MED_Chaos/Meshes/Parts/F_MED_Chaos_FaceAcc_AnimBP
target_skeleton=&/F_MED_Chaos_FaceAcc_Skeleton
graphs [
 AnimGraph {nodes [
   n1 = Get RBAN OFF() -> (RBAN OFF:double)
   "Transform (Modify) Bone - Bone: Dyn_RightBack_Hair4"(Alpha=@n1, Translation="0,0,1", Rotation="0,2.5,0", Scale="1,1,1") {
    rotationMode=BMM_Replace
    rotationSpace=BCS_BoneSpace
    translationMode=BMM_Additive
    translationSpace=BCS_BoneSpace
    alphaBoolBlend {blendOption=Linear}
    alphaScaleBiasClamp {bInterpResult=true}
    boneToModify {boneName=Dyn_RightBack_Hair4}
    rotation {yaw=2.5}
    translation {z=1}
   }
   "Transform (Modify) Bone - Bone: Dyn_LeftBack_Hair4"(Alpha=@n1, Translation="0,0,1", Rotation="0,2.5,0", Scale="1,1,1") {
    rotationMode=BMM_Replace
    rotationSpace=BCS_BoneSpace
    translationMode=BMM_Additive
    translationSpace=BCS_BoneSpace
    alphaBoolBlend {blendOption=Linear}
    alphaScaleBiasClamp {bInterpResult=true}
    boneToModify {boneName=Dyn_LeftBack_Hair4}
    rotation {yaw=2.5}
    translation {z=1}
   }
   n2 = "Constraint - Dyn_Master_RightBack_Hair (RR)"!(ComponentPose=Local To Component(LocalPose=Use cached pose 'start_cache'()), Alpha=1, ConstraintWeights_0=0.8, ConstraintWeights_1:float=0.2) -> (Pose) {
    alphaBoolBlend {blendOption=Linear}
    boneToModify {boneName=Dyn_Master_RightBack_Hair}
    constraintSetup [{transformType=Rotation, targetBone {boneName=spine_05}}, {transformType=Rotation, targetBone {boneName=neck_01}}]
    constraintWeights [1,1]
   }
   n3 = "Constraint - Dyn_Master_LeftBack_Hair (RR)"!(ComponentPose=@n2, Alpha=1, ConstraintWeights_0=0.8, ConstraintWeights_1:float=0.2) -> (Pose) {
    alphaBoolBlend {blendOption=Linear}
    boneToModify {boneName=Dyn_Master_LeftBack_Hair}
    constraintSetup [{transformType=Rotation, targetBone {boneName=spine_05}}, {transformType=Rotation, targetBone {boneName=neck_01}}]
    constraintWeights [1,1]
   }
   n4 = "Constraint - Dyn_Master_MiddleBack_Hair (RR)"!(ComponentPose=@n3, Alpha=1, ConstraintWeights_0=0.8, ConstraintWeights_1:float=0.2) -> (Pose) {
    alphaBoolBlend {blendOption=Linear}
    boneToModify {boneName=Dyn_Master_MiddleBack_Hair}
    constraintSetup [{transformType=Rotation, targetBone {boneName=spine_05}}, {transformType=Rotation, targetBone {boneName=neck_01}}]
    constraintWeights [1,1]
   }
   n5 = Get ADS_UP() -> (ADS_UP:double)
   n6 = "Transform (Modify) Bone - Bone: Dyn_Master_RB_Hair2_Side"(ComponentPose=@n4, Alpha=@n5, Translation="-1,-4,1", Rotation="0,30,50", Scale="1,1,1") {
    rotationMode=BMM_Additive
    rotationSpace=BCS_BoneSpace
    translationMode=BMM_Additive
    translationSpace=BCS_BoneSpace
    alphaBoolBlend {blendOption=Linear}
    alphaScaleBiasClamp {bInterpResult=true, interpSpeedDecreasing=20, interpSpeedIncreasing=50}
    boneToModify {boneName=Dyn_Master_RB_Hair2_Side}
    rotation {roll=50, yaw=30}
    translation=<-1 -4 1>
   }
   n7 = "Transform (Modify) Bone - Bone: Dyn_Master_LB_Hair2_Side"(ComponentPose=@n6, Alpha=@n5, Translation="0,-2,-1.5", Rotation="0,0,0", Scale="1,1,1") {
    rotationMode=BMM_Additive
    rotationSpace=BCS_BoneSpace
    translationMode=BMM_Additive
    translationSpace=BCS_BoneSpace
    alphaBoolBlend {blendOption=Linear}
    alphaScaleBiasClamp {bInterpResult=true, interpSpeedIncreasing=20}
    boneToModify {boneName=Dyn_Master_LB_Hair2_Side}
    translation {y=-2, z=-1.5}
   }
   n8 = "Transform (Modify) Bone - Bone: Dyn_Master_RB_Hair2_Side"!(ComponentPose=@n7, bAlphaBoolEnabled:bool=Get bIsTargeting(), Translation="0,0,0", Rotation="0,0,50", Scale="1,1,1") -> (Pose) {
    alphaInputType=Bool
    rotationMode=BMM_Additive
    rotationSpace=BCS_BoneSpace
    translationMode=BMM_Additive
    translationSpace=BCS_BoneSpace
    alphaBoolBlend {blendInTime=0.2, blendOption=Linear, blendOutTime=0.2}
    alphaScaleBiasClamp {bInterpResult=true, interpSpeedDecreasing=20, interpSpeedIncreasing=50}
    boneToModify {boneName=Dyn_Master_RB_Hair2_Side}
    rotation {roll=50}
   }
   n9 = RigidBody(ComponentPose=@n8, Alpha=Get Is_Wearing Backbling() -> (Is_Wearing Backbling:double)) {
    bAllowPoseScalePropagation=false
    bClampLinearTranslationLimitToRefPose=true
    bOverrideWorldGravity=true
    overridePhysicsAsset=&/F_MED_Chaos_FaceAcc_Backpack_Physics
    simulationSpace=BaseBoneSpace
    alphaBoolBlend {blendOption=Linear}
    baseBoneRef {boneName=spine_05}
    componentAppliedLinearAccClamp=<500 500 500>
    componentLinearAccScale=<0.05 0.05 0.05>
    componentLinearVelScale=<0.05 0.05 0.05>
    overrideWorldGravity {z=-200}
    simSpaceSettings {velocityScaleZ=0.5}
   }
   n10 = RigidBody(ComponentPose=@n9, Alpha=Get Is_not_Wearing Backbling() -> (Is_not_Wearing Backbling:double)) {
    bClampLinearTranslationLimitToRefPose=true
    bOverrideWorldGravity=true
    overridePhysicsAsset=&/F_MED_Chaos_FaceAcc_Physics
    simulationSpace=BaseBoneSpace
    alphaBoolBlend {blendOption=Linear}
    baseBoneRef {boneName=head}
    componentAppliedLinearAccClamp=<3000 3000 3000>
    componentLinearAccScale=<0.4 0.4 0.4>
    componentLinearVelScale=<0.4 0.4 0.4>
    overrideWorldGravity {z=-980}
   }
   n11 = Get Is_Wearing Backbling()
   n12 = "Constraint - Dyn_MiddleBack_Hair1 (PP)"!(ComponentPose=@n10, Alpha=@n11, ConstraintWeights_0=0.5, ConstraintWeights_1:float=0.5) -> (Pose) {
    alphaBoolBlend {blendOption=Linear}
    boneToModify {boneName=Dyn_MiddleBack_Hair1}
    constraintSetup [{targetBone {boneName=neck_02}}, {targetBone {boneName=head}}]
    constraintWeights [1,1]
   }
   n13 = "Constraint - Dyn_LeftBack_Hair1 (PP)"!(ComponentPose=@n12, Alpha=@n1, ConstraintWeights_0=0.5, ConstraintWeights_1:float=0.5) -> (Pose) {
    alphaBoolBlend {blendOption=Linear}
    boneToModify {boneName=Dyn_LeftBack_Hair1}
    constraintSetup [{targetBone {boneName=neck_02}}, {targetBone {boneName=head}}]
    constraintWeights [1,1]
   }
   n14 = "Constraint - Dyn_RightBack_Hair1 (PP)"!(ComponentPose=@n13, Alpha=@n1, ConstraintWeights_0=0.5, ConstraintWeights_1:float=0.5) -> (Pose) {
    alphaBoolBlend {blendOption=Linear}
    boneToModify {boneName=Dyn_RightBack_Hair1}
    constraintSetup [{targetBone {boneName=neck_02}}, {targetBone {boneName=head}}]
    constraintWeights [1,1]
   }
   n15 = "Constraint - Dyn_MiddleBack_Hair2 (PP)"!(ComponentPose=@n14, Alpha=@n11, ConstraintWeights_0=0.6, ConstraintWeights_1:float=0.4) -> (Pose) {
    alphaBoolBlend {blendOption=Linear}
    boneToModify {boneName=Dyn_MiddleBack_Hair2}
    constraintSetup [{targetBone {boneName=neck_02}}, {targetBone {boneName=spine_05}}]
    constraintWeights [1,1]
   }
   n16 = "Constraint - Dyn_LeftBack_Hair2 (PP)"!(ComponentPose=@n15, Alpha=@n1, ConstraintWeights_0=0.6, ConstraintWeights_1:float=0.4) -> (Pose) {
    alphaBoolBlend {blendOption=Linear}
    boneToModify {boneName=Dyn_LeftBack_Hair2}
    constraintSetup [{targetBone {boneName=neck_02}}, {targetBone {boneName=spine_05}}]
    constraintWeights [1,1]
   }
   n17 = "Constraint - Dyn_RightBack_Hair2 (PP)"!(ComponentPose=@n16, Alpha=@n1, ConstraintWeights_0=0.6, ConstraintWeights_1:float=0.4) -> (Pose) {
    alphaBoolBlend {blendOption=Linear}
    boneToModify {boneName=Dyn_RightBack_Hair2}
    constraintSetup [{targetBone {boneName=neck_02}}, {targetBone {boneName=spine_05}}]
    constraintWeights [1,1]
   }
   n18 = "Constraint - Dyn_MiddleBack_Hair3 (PP)"!(ComponentPose=@n17, Alpha=@n11, ConstraintWeights_0=0.5, ConstraintWeights_1:float=0.5) -> (Pose) {
    alphaBoolBlend {blendOption=Linear}
    boneToModify {boneName=Dyn_MiddleBack_Hair3}
    constraintSetup [{targetBone {boneName=neck_02}}, {targetBone {boneName=spine_05}}]
    constraintWeights [1,1]
   }
   n19 = "Constraint - Dyn_LeftBack_Hair3 (PP)"!(ComponentPose=@n18, Alpha=@n1, ConstraintWeights_0=0.5, ConstraintWeights_1:float=0.5) -> (Pose) {
    alphaBoolBlend {blendOption=Linear}
    boneToModify {boneName=Dyn_LeftBack_Hair3}
    constraintSetup [{targetBone {boneName=clavicle_l}}, {targetBone {boneName=spine_05}}]
    constraintWeights [1,1]
   }
   n20 = "Constraint - Dyn_RightBack_Hair3 (PP)"!(ComponentPose=@n19, Alpha=@n1, ConstraintWeights_0=0.5, ConstraintWeights_1:float=0.5) -> (Pose) {
    alphaBoolBlend {blendOption=Linear}
    boneToModify {boneName=Dyn_RightBack_Hair3}
    constraintSetup [{targetBone {boneName=clavicle_r}}, {targetBone {boneName=spine_05}}]
    constraintWeights [1,1]
   }
   n21 = "Constraint - Dyn_MiddleBack_Hair4 (PP)"!(ComponentPose=@n20, Alpha=@n11, ConstraintWeights_0=0.5, ConstraintWeights_1:float=0.5) -> (Pose) {
    alphaBoolBlend {blendOption=Linear}
    boneToModify {boneName=Dyn_MiddleBack_Hair4}
    constraintSetup [{targetBone {boneName=spine_04}}, {targetBone {boneName=spine_05}}]
    constraintWeights [1,1]
   }
   n22 = "Constraint - Dyn_LeftBack_Hair4 (PP)"!(ComponentPose=@n21, Alpha=@n1, ConstraintWeights_0=0.5, ConstraintWeights_1:float=0.5) -> (Pose) {
    alphaBoolBlend {blendOption=Linear}
    boneToModify {boneName=Dyn_LeftBack_Hair4}
    constraintSetup [{targetBone {boneName=clavicle_l}}, {targetBone {boneName=spine_05}}]
    constraintWeights [1,1]
   }
   n23 = "Constraint - Dyn_RightBack_Hair4 (PP)"!(ComponentPose=@n22, Alpha=@n1, ConstraintWeights_0=0.5, ConstraintWeights_1:float=0.5) -> (Pose) {
    alphaBoolBlend {blendOption=Linear}
    boneToModify {boneName=Dyn_RightBack_Hair4}
    constraintSetup [{targetBone {boneName=spine_03}}, {targetBone {boneName=clavicle_r}}]
    constraintWeights [1,1]
   }
   n24 = "Transform (Modify) Bone - Bone: Dyn_MiddleBack_Hair2"(ComponentPose=@n23, Alpha=@n11, Translation="0,0,0", Rotation="0,0,0", Scale="1,1,1") {
    rotationMode=BMM_Replace
    rotationSpace=BCS_BoneSpace
    translationMode=BMM_Additive
    translationSpace=BCS_BoneSpace
    alphaBoolBlend {blendOption=Linear}
    alphaScaleBiasClamp {bInterpResult=true}
    boneToModify {boneName=Dyn_MiddleBack_Hair2}
   }
   n25 = "Transform (Modify) Bone - Bone: Dyn_LeftBack_Hair2"(ComponentPose=@n24, Alpha=@n1, Translation="-2,-3,0", Rotation="0,0,0", Scale="1,1,1") {
    rotationMode=BMM_Replace
    rotationSpace=BCS_BoneSpace
    translationMode=BMM_Additive
    translationSpace=BCS_BoneSpace
    alphaBoolBlend {blendOption=Linear}
    alphaScaleBiasClamp {bInterpResult=true}
    boneToModify {boneName=Dyn_LeftBack_Hair2}
    translation=<-2 -3>
   }
   n26 = "Transform (Modify) Bone - Bone: Dyn_RightBack_Hair2"(ComponentPose=@n25, Alpha=@n1, Translation="0,-1,0", Rotation="0,0,0", Scale="1,1,1") {
    rotationMode=BMM_Replace
    rotationSpace=BCS_BoneSpace
    translationMode=BMM_Additive
    translationSpace=BCS_BoneSpace
    alphaBoolBlend {blendOption=Linear}
    alphaScaleBiasClamp {bInterpResult=true}
    boneToModify {boneName=Dyn_RightBack_Hair2}
    translation {y=-1}
   }
   n27 = Get is skydiving() -> (is skydiving:double)
   n28 = "Transform (Modify) Bone - Bone: Dyn_LeftBack_Hair2"(ComponentPose=@n26, Alpha=@n27, Translation="3.213,-0.423,-4.466", Rotation="1.642,22.344,-11.141", Scale="1,1,1") {
    rotationMode=BMM_Replace
    rotationSpace=BCS_BoneSpace
    translationMode=BMM_Additive
    translationSpace=BCS_BoneSpace
    alphaBoolBlend {blendOption=Linear}
    alphaScaleBiasClamp {bInterpResult=true}
    boneToModify {boneName=Dyn_LeftBack_Hair2}
    rotation=<rot 1.6421 22.3441 -11.141>
    translation=<3.213 -0.423 -4.466>
   }
   n29 = "Transform (Modify) Bone - Bone: Dyn_RightBack_Hair2"(ComponentPose=@n28, Alpha=@n27, Translation="1.776,-1.609,2.297", Rotation="-0.000,11.253,-19.688", Scale="1,1,1") {
    rotationMode=BMM_Replace
    rotationSpace=BCS_BoneSpace
    translationMode=BMM_Additive
    translationSpace=BCS_BoneSpace
    alphaBoolBlend {blendOption=Linear}
    alphaScaleBiasClamp {bInterpResult=true}
    boneToModify {boneName=Dyn_RightBack_Hair2}
    rotation=<rot 0 11.253 -19.688>
    translation=<1.7757 -1.609 2.297>
   }
   n30 = "Transform (Modify) Bone - Bone: Dyn_MiddleBack_Hair3"(ComponentPose=@n29, Alpha=@n11, Translation="0,0,1", Rotation="0,5,0", Scale="1,1,1") {
    rotationMode=BMM_Replace
    rotationSpace=BCS_BoneSpace
    translationMode=BMM_Additive
    translationSpace=BCS_BoneSpace
    alphaBoolBlend {blendOption=Linear}
    alphaScaleBiasClamp {bInterpResult=true}
    boneToModify {boneName=Dyn_MiddleBack_Hair3}
    rotation {yaw=5}
    translation {z=1}
   }
   n31 = "Transform (Modify) Bone - Bone: Dyn_LeftBack_Hair3"(ComponentPose=@n30, Alpha=@n1, Translation="-2,-1.5,-1.5", Rotation="0,5,0", Scale="1,1,1") {
    rotationMode=BMM_Replace
    rotationSpace=BCS_BoneSpace
    translationMode=BMM_Additive
    translationSpace=BCS_BoneSpace
    alphaBoolBlend {blendOption=Linear}
    alphaScaleBiasClamp {bInterpResult=true}
    boneToModify {boneName=Dyn_LeftBack_Hair3}
    rotation {yaw=5}
    translation=<-2 -1.5 -1.5>
   }
   n32 = "Transform (Modify) Bone - Bone: Dyn_RightBack_Hair3"(ComponentPose=@n31, Alpha=@n1, Translation="-2,-1,1", Rotation="0,5,0", Scale="1,1,1") {
    rotationMode=BMM_Replace
    rotationSpace=BCS_BoneSpace
    translationMode=BMM_Additive
    translationSpace=BCS_BoneSpace
    alphaBoolBlend {blendOption=Linear}
    alphaScaleBiasClamp {bInterpResult=true}
    boneToModify {boneName=Dyn_RightBack_Hair3}
    rotation {yaw=5}
    translation=<-2 -1 1>
   }
   n33 = "Transform (Modify) Bone - Bone: Dyn_MiddleBack_Hair4"(ComponentPose=@n32, Alpha=@n11, Translation="0,0,1", Rotation="0,2.5,0", Scale="1,1,1") {
    rotationMode=BMM_Replace
    rotationSpace=BCS_BoneSpace
    translationMode=BMM_Additive
    translationSpace=BCS_BoneSpace
    alphaBoolBlend {blendOption=Linear}
    alphaScaleBiasClamp {bInterpResult=true}
    boneToModify {boneName=Dyn_MiddleBack_Hair4}
    rotation {yaw=2.5}
    translation {z=1}
   }
   Output Pose(Result=Component To Local(ComponentPose=@n33)) {layerGroup=None}
   n34 = Layered blend per bone(
    BasePose=Copy Pose From Mesh() {bCopyCurves=true, bUseAttachedParent=true}
    BlendPoses_0=Copy Pose From Mesh(SourceMeshComponent=Get Body_Mesh())
   ) {layerSetup [{branchFilters [{boneName=clavicle_l}, {boneName=clavicle_r}, {boneName=upperarm_r}, {boneName=upperarm_l}]}]}
   Save cached pose 'start_cache'(Pose=@n34) {cacheName=start_cache}
   Get wearingBackpack() -> (wearingBackpack:bool)
   Get noBackpack() -> (noBackpack:bool)
  ]}
 EventGraph {nodes [
   Event BlueprintUpdateAnimation()
   > Set ADS_UP(ADS_UP:double=GetAlphaFromReferenceTransformDelta(SkeletalMeshComponent=Get Skel Mesh(), BoneName=head, BaseName=neck_02, InputRange=(X=0,Y=-25), TwistAxis="0,0,1")) -> (Output_Get:double)
   > n1 = Set RBAN OFF_0(RBAN OFF_0:bool=NOT Boolean(A=Get bRigidBodyEnabled())) -> (Output_Get:bool)
   > Set RBAN OFF(RBAN OFF:double="To Float (Boolean)"(InBool=@n1)) -> (Output_Get:double)
   n2 = OR Boolean(A=Get bIsSkyDiving(), B=Get bIsPlayingEmote())
   n3 = Get RBAN OFF_0() -> (RBAN OFF_0:bool)
   > Set is skydiving(is skydiving:double="To Float (Boolean)"(InBool=AND Boolean(A=@n3, B=@n2))) -> (Output_Get:double)
   Event OnCharacterCustomizationLoaded()
   > n4 = Sequence!(execute) -> (then_0, then_1, then_2:Exec)
   > n5 = Branch(Condition=CheckForBackbling())
   > Set Is_not_Wearing Backbling(Is_not_Wearing Backbling:double=0) -> (Output_Get:double)
   > Set Is_Wearing Backbling(Is_Wearing Backbling:double=1) -> (Output_Get:double)
   Set Is_Wearing Backbling(execute=@n5.else, Is_Wearing Backbling=0)
   > Set Is_not_Wearing Backbling(Is_not_Wearing Backbling=1)
   n6 = Cast To PlayerPawn_Athena(execute=@n4.then_1, Object=TryGetPawnOwner())
   > Set Body_Mesh(Body_Mesh=GetSkeletalMeshForPartType(self=@n6))
   Event BlueprintInitializeAnimation()
   > n7 = Sequence()
   > Set Skel Mesh(Skel Mesh=GetOwnerSkeletalMeshComponent())
   n8 = CheckForBackbling()
   n9 = Set wearingBackpack(execute=@n7.then_1, wearingBackpack:bool=AND Boolean(A=@n8, B=NOT Boolean(A=@n8.bIsBackblingACape))) -> (Output_Get:bool)
   > Set noBackpack(noBackpack:bool=NOT Boolean(A=@n9)) -> (Output_Get:bool)
   GetOwningActor()
  ]}
]
variables [
 ADS_UP {type=number}
 Body_Mesh {type=SkeletalMeshComponent}
 is skydiving {type=number}
 Is_not_Wearing Backbling {type=number}
 Is_Wearing Backbling {type=number}
 noBackpack {type=boolean}
 RBAN OFF {type=number}
 RBAN OFF_0 {type=boolean}
 Skel Mesh {type=SkeletalMeshComponent}
 wearingBackpack {type=boolean, default=true}
]
